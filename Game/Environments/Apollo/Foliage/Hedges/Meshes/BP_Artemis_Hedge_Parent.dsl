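; Parent Blueprint for Artemis hedge foliage actors that applies a snow coverage amount to the hedge mesh.
; Both the construction script and BeginPlay call the parent, then, only where cosmetic events can run, write SnowCoverageAmount into custom primitive data slot 6 on the static mesh component.
; Overlap and tick events are present but empty.
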(event EventBeginPlay
  (|CallParentFunction)
  (Networking|CanExecuteCosmeticEvents
    (:True
      (Rendering|Material|SetCustomPrimitiveDataFloat (Variables|Debug|GetStaticMeshComponent) 6 (Variables|Default|GetSnowCoverageAmount)))))

(event Collision|EventActorBeginOverlap
  (params (OtherActor :type Actor%20Object%20Reference)))

(event EventTick
  (params (DeltaSeconds :type Float%20%28single-precision%29)))

(fn ConstructionScript
  (|CallParentFunction)
  (Networking|CanExecuteCosmeticEvents
    (:True
      (Rendering|Material|SetCustomPrimitiveDataFloat (Variables|Debug|GetStaticMeshComponent) 6 (Variables|Default|GetSnowCoverageAmount)))))
